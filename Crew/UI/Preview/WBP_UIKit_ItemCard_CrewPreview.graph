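# Crew preview item card widget, derived from WBP_UIKit_ItemCard_Base, representing a locker item in a list.
# Hover plays a slight scale-up animation and, like selection, pushes the item to FortItemSelectionVM, while clicks select it in the owning ListView.
# Touch input gets special focus and UIKit_Transference visibility handling.
# Helper functions mark items seen and drive secondary equipped-slot indicators from category config; context-menu and double-click functions are empty.
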WidgetBlueprint /Crew/UI/Preview/WBP_UIKit_ItemCard_CrewPreview
animations [Anim_Hovered {
  binding_count=1
  length_seconds=0.2004
  object=&self:Anim_Hovered
  sequence {bindings [WBP_UIKit_ItemCard_Locker_C_17 {
     id=#1
     possessed_class=WBP_UIKit_ItemCard_Locker_C
     tracks [{
       type=MovieScene2DTransformTrack
       sections [{
         type=MovieScene2DTransformSection
         end_frame=4
         start_frame=0
         channels [
          Translation.X {type=MovieSceneScriptingFloatChannel, const [0,1]}
          Translation.Y {type=MovieSceneScriptingFloatChannel, const [0,1]}
          Angle {type=MovieSceneScriptingFloatChannel, const [0,1]}
          Scale.X {type=MovieSceneScriptingFloatChannel, keys [[0,1],[4.007,1.03]]}
          Scale.Y {type=MovieSceneScriptingFloatChannel, keys [[0,1],[4.007,1.03]]}
          Shear.X {type=MovieSceneScriptingFloatChannel, const [0,1]}
          Shear.Y {type=MovieSceneScriptingFloatChannel, const [0,1]}
         ]
        }]
      }]
    }]}
 }]
graphs [
 EventGraph {nodes [
   n1 = Event PreConstruct()
   > Parent: PreConstruct(IsDesignTime=@n1.IsDesignTime)
   > SetGridAlignment!(
    execute
    self:WBP UIKit Item Card Base Object Reference
    WIdget:Widget Object Reference=Get Slot_Main() -> (Slot_Main:Named Slot Object Reference)
    Horizontal:EHorizontalAlignment Enum=HAlign_Right
    Vertical:EVerticalAlignment Enum=VAlign_Bottom
   ) -> (then)
   Event OnListItemObjectSet()
   > InitializeItemCard(self:WBP UIKit Item Card Base Object Reference)
   Event On Deselected()
   > Branch(Condition=Get IsButtonFocused() -> (IsButtonFocused:bool))
   > Parent: On Deselected() {is=title K2Node_CallParentFunction}
   > Branch(Condition=IsUsingTouch(PlayerController=GetOwningPlayer()))
   > SetVisibility(self=Get UIKit_Transference() -> (UIKit_Transference:WBP UIKit Transference Object Reference), InVisibility=SelfHitTestInvisible)
   Event On Clicked()
   > Parent: On Clicked() {is=title K2Node_CallParentFunction}
   n2 = GetOwningPlayer()
   n3 = GetSelected!(self:Common Button Base Object Reference) -> (ReturnValue:bool)
   > n4 = Branch(Condition=AND Boolean(A=NOT Boolean(A=@n3), B=IsUsingTouch(PlayerController=@n2)))
   > PlayAnimationForward(InAnimation=Get Anim_Hovered() -> (Anim_Hovered:Widget Animation Object Reference))
   > n5 = SetFocusState!(execute, self:WBP UIKit Item Card Base Object Reference, IsFocused:bool=true) -> (then)
   n6 = Cast To ListView!(
    execute=[@n5.then, @n4.else]
    Object=GetOwningListView!(UserListEntry:User List Entry Interface) -> (ReturnValue:List View Base Object Reference)
   ) -> (then, CastFailed, AsList View:List View Object Reference) {is=title K2Node_DynamicCast}
   > Set Selected Item!(
    execute
    self:List View Object Reference=@n6
    Item:Object Reference=GetListItemObject!(UserObjectListEntry:User Object List Entry Interface) -> (ReturnValue:Object Reference)
   ) -> (then) {is=title K2Node_CallFunction}
   Event On Hovered()
   > Set IsItemCardLockerHovered(IsItemCardLockerHovered:bool=true) -> (Output_Get:bool)
   > Set SelectedItem!(
    execute
    SelectedItem:Fort Item VM Object Reference=Get FortItemVM() -> (FortItemVM:Fort Item VM Object Reference)
    self:Fort Item Selection VM Object Reference=Get FortItemSelectionVM() -> (FortItemSelectionVM:Fort Item Selection VM Object Reference)
   ) -> (then, Output_Get:Fort Item VM Object Reference)
   > Parent: On Hovered() {is=title K2Node_CallParentFunction}
   > PlayAnimationForward(InAnimation=Get Anim_Hovered())
   Event On Unhovered()
   > Set IsItemCardLockerHovered(IsItemCardLockerHovered=false)
   > Parent: On Unhovered() {is=title K2Node_CallParentFunction}
   > PlayAnimationReverse(InAnimation=Get Anim_Hovered())
   Event On Selected()
   > Parent: On Selected() {is=title K2Node_CallParentFunction}
   > Set SelectedItem(SelectedItem=Get FortItemVM(), self=Get FortItemSelectionVM())
   > Branch(Condition=IsUsingTouch(PlayerController=GetOwningPlayer()))
   > SetVisibility(self=Get UIKit_Transference(), InVisibility=Hidden)
   n7 = Event On Input Method Changed!() -> (OutputDelegate, then, CurrentInputType:ECommonInputType Enum) {is=title K2Node_Event}
   > Is Valid(InputObject=Get FortItemVM())
   > n8 = Switch on ECommonInputType!(execute, Selection:ECommonInputType Enum=@n7.CurrentInputType) -> (MouseAndKeyboard:Exec, Gamepad:Exec, Touch:Exec, Count:Exec) {is=title K2Node_SwitchEnum}
   > Branch(Condition=GetSelected())
   > SetFocusState(IsFocused=false)
   Branch(execute=@n8.Touch, Condition=GetSelected())
   > SetFocusState(IsFocused=true)
   n9 = Event OnAddedToFocusPath!() -> (OutputDelegate, then, InFocusEvent:Focus Event Structure) {is=title K2Node_Event}
   > Parent: OnAddedToFocusPath!(execute, InFocusEvent:Focus Event Structure=@n9.InFocusEvent) -> (then) {is=title K2Node_CallParentFunction}
   > n10 = Switch on ECommonInputType(Selection=GetCurrentInputType(self=Get CommonInputSubsystem()))
   SetFocusState(execute=@n10.Touch, IsFocused=GetSelected())
  ]}
 GoToLockerItemDetails {nodes [GoToLockerItemDetails!() -> (then, CustomItemDetails:User Widget Class Reference) {is=title K2Node_FunctionEntry}]}
 Is Item Equipped {nodes [
   n1 = Is Item Equipped!() -> (then, Category Item:Fort Item VM Object Reference, Selected Item:Fort Item VM Object Reference) {is=title K2Node_FunctionEntry}
   > Return Node!(execute, NewParam:bool="Equal (Object)"(A=@n1, B=@n1.Selected Item))
  ]}
 OnContextMenuClosed {nodes [OnContextMenuClosed() {is=title K2Node_FunctionEntry}]}
 OnContextMenuEditStylesClicked {nodes [OnContextMenuEditStylesClicked() {is=title K2Node_FunctionEntry}]}
 OnContextMenuOpened {nodes [OnContextMenuOpened!() -> (then, ContextMenuWidgetParentEntry:Widget Object Reference) {is=title K2Node_FunctionEntry}]}
 OnItemCardHovered {nodes [OnItemCardHovered() {is=title K2Node_FunctionEntry}]}
 OnItemDoubleClicked {nodes [OnItemDoubleClicked() {is=title K2Node_FunctionEntry}]}
 SelectAndMarkSeen {nodes [
   SelectAndMarkSeen() {is=title K2Node_FunctionEntry}
   n1 = Get FortItemSelectionVM()
   > Set SelectedItem(SelectedItem=Get FortItemVM(), self=@n1)
   n2 = Get SelectedCategory!(self:Fort Item Selection VM Object Reference=@n1) -> (SelectedCategory:Fort Locker Category Item VM Object Reference)
   > Is Valid(InputObject=@n2)
   > MarkItemAsSeen!(
    execute
    self:Fort Locker Category VM Object Reference=Get Category!(self:Fort Locker Category Item VM Object Reference=@n2) -> (Category:Fort Locker Category VM Object Reference)
    Item:Fort Item VM Object Reference=Get FortItemVM()
   ) -> (then)
  ]}
 SelectCard {nodes [
   SelectCard() {is=title K2Node_FunctionEntry}
   > PlayAnimationForward(InAnimation=Get Anim_Hovered())
   > SetFocusState(IsFocused=true)
   > n1 = Cast To ListView(Object=GetOwningListView())
   > Set Selected Item(self=@n1, Item=GetListItemObject())
  ]}
 Update Equipped Indexes {nodes [
   n1 = Update Equipped Indexes!() -> (then, Equipped Indexes:Array of Integers) {is=title K2Node_FunctionEntry}
   > n2 = Branch(Condition=Is Empty!(TargetArray:Array of Integers=@n1) -> (ReturnValue:bool))
   > SetItemIsEquippedSecondary!(
    execute
    self:WBP UIKit Item Card Base Object Reference
    Is Equipped:bool=false
    ShowEquippedSlotIndex:bool=false
    EquippedSlotIndex:int=0
    SlotIcon:Soft Object Reference
   ) -> (then)
   n3 = GetItemCategory!(self:Fort Item VM Object Reference=Get FortItemVM()) -> (ReturnValue:Fort Locker Category Config Object Reference)
   n4 = "Get (a copy)"(Array:Array of Integers=Get Equipped Indexes() -> (Equipped Indexes:Array of Integers)) -> (Output:int)
   SetItemIsEquippedSecondary(
    execute=@n2.else
    Is Equipped=true
    ShowEquippedSlotIndex=Get bSecondaryEquipShowNumbers!(self:Fort Locker Category Config Object Reference=@n3) -> (bSecondaryEquipShowNumbers:bool)
    EquippedSlotIndex=int + int(A=@n4)
    SlotIcon=Break Fort Locker Category Slot Info!(FortLockerCategorySlotInfo:"Fort Locker Category Slot Info Structure (by ref)"=GetSlotInfo!(self:Fort Locker Category Config Object Reference=@n3, SlotIndex:int=@n4) -> (ReturnValue:"Fort Locker Category Slot Info Structure (by ref)")) -> (
     CategoryName:Text
     CategoryNamePlural:Text
     CategoryNameWithPluralForms:Text
     CategoryDescription:Text
     DefaultEmptyTexture:Soft Object Reference
     OverrideShuffleTexture:Texture 2D Soft Object Reference
     CategoryIcon:Material Instance Object Reference
     CategorySoundOverrides:Map of Gameplay Tag Structures to Sound Base Soft Object References
     bShowInCategoryTab:bool=~
     PreviewWindowScale:float=~
     PreviewWindowOffset:Vector 3f
     bUsePluralCategoryInItemList:bool=~
    ) {is=title K2Node_BreakStruct}.DefaultEmptyTexture
   )
  ]}
]
variables [
 AwaitingContextMenuClose {type=boolean}
 ContextMenuActionContext {type=UIKitContextMenuActionContext, default=null}
 GoToLockerItemDetails {}
 IsItemCardLockerHovered {type=boolean}
 OnContextMenuClosed {}
 OnContextMenuEditStylesClicked {}
 OnContextMenuOpened {}
 OnItemCardHovered {}
 OnItemDoubleClicked {}
]
widgets {
 info {
  inheritedWidgetCount=0
  parentClass=/FortUILibrary/Systems/Items/WBP_UIKit_ItemCard_Base.WBP_UIKit_ItemCard_Base_C
  rootWidgetClass=null
  widgetCount=0
 }
 widgets []
}
