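# Fortnite Crew preview screen, an activatable widget showing subscription rewards in a 3D item preview with a hold-to-purchase button.
# On activation, rewards fill a tile view and select the first, while a switcher opens the grid, or the style selector for a single reward.
# Back returns to the grid or pops the screen.
# Async purchase disables the widget until completion and broadcasts Subscribe Completed on success; upsell music functions are empty stubs.
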WidgetBlueprint /Crew/UI/Preview/CrewItemPreviewSreen
graphs [
 BP_GetDesiredFocusTarget {nodes [
   Get Desired Focus Target()
   n1 = GetActiveWidget!(self:Common Visibility Switcher Object Reference=Get Switcher_ScreenView() -> (Switcher_ScreenView:Common Visibility Switcher Object Reference)) -> (ReturnValue:Widget Object Reference)
   n2 = Get WBP_UIKit_VariantSelector() -> (WBP_UIKit_VariantSelector:WBP UIKit Variant Selector Object Reference)
   n3 = Get VBox_VariantSelector() -> (VBox_VariantSelector:Vertical Box Object Reference)
   > Return Node!(
    execute
    ReturnValue:Widget Object Reference=Select(
     Option 0:Widget Object Reference=@n1
     Option 1:Widget Object Reference=GetDesiredFocusTarget!(self:WBP UIKit Variant Selector Object Reference=@n2) -> (ReturnValue:Widget Object Reference)
     Index="Equal (Object)"(A=@n1, B=@n3)
    ) -> (ReturnValue:Widget Object Reference)
   )
  ]}
 BP_GetDesiredInputConfig {nodes [
   Get Desired Input Config() {is=title K2Node_FunctionEntry}
   > Return Node(ReturnValue:UIInput Config Structure=Make UIInput Config!(
     bIgnoreMoveInput:bool=False
     bIgnoreLookInput:bool=False
     InputMode:ECommonInputMode Enum=All
     MouseCaptureMode:EMouseCaptureMode Enum=CaptureDuringMouseDown
     MouseLockMode:EMouseLockMode Enum=DoNotLock
     bHideCursorDuringViewportCapture:bool=True
    ) -> (UIInputConfig:UIInput Config Structure) {is=title K2Node_MakeStruct})
  ]}
 EventGraph {nodes [
   "On Clicked (ConfirmPurchase_Button)"!() -> (OutputDelegate, then, Button:Common Button Base Object Reference) {is=title K2Node_ComponentBoundEvent}
   > n1 = PurchaseSubscription!(execute:Exec) -> (then:Exec, OnOperationCompleted:Exec, Success:bool=~) {is=title K2Node_AsyncAction}
   > SetIsEnabled(bInIsEnabled=false)
   SetIsEnabled(execute=@n1.OnOperationCompleted, bInIsEnabled=true)
   > Branch(Condition=@n1)
   > Call Subscribe Completed!(execute:Exec, self:Crew Item Preview Sreen Object Reference) -> (then:Exec) {is=title K2Node_CallDelegate}
   > UpdateUpsellMusic()
   n2 = "On Clicked (Button_ItemDetails)"!() -> (OutputDelegate, then, Button:Common Button Base Object Reference) {is=title K2Node_ComponentBoundEvent}
   > Fire Event Element Interacted(Element=@n2.Button)
   > Switch Screen!(execute, self, Widget:Widget Object Reference=Get VBox_VariantSelector()) -> (then)
   > Set FortItemVM!(
    execute
    FortItemVM:Fort Item VM Object Reference=Get SelectedItem!(self:Fort Item Selection VM Object Reference=Get FortItemSelectionVM() -> (FortItemSelectionVM:Fort Item Selection VM Object Reference)) -> (SelectedItem:Fort Item VM Object Reference)
    self:WBP UIKit Variant Selector Object Reference=Get WBP_UIKit_VariantSelector()
   ) -> (then, Output_Get:Fort Item VM Object Reference)
   "On Clicked (Button_Back)"!() -> (OutputDelegate, then, Button:Common Button Base Object Reference) {is=title K2Node_ComponentBoundEvent}
   n3 = GetActiveWidget(self=Get Switcher_ScreenView())
   n4 = Get VBox_VariantSelector()
   n5 = Length(TargetArray:Array of Fort Item VM Object References=Get Rewards() -> (Rewards:Array of Fort Item VM Object References))
   > n6 = Branch(Condition=AND Boolean(A="integer > integer"(A=@n5, B=1), B="Equal (Object)"(A=@n3, B=@n4)))
   n7 = Self-Reference()
   > FireEventElementInteractedByName!(execute, Element Name:String=Back To Item View, WidgetCallingFrom:Widget Object Reference=@n7) -> (then, Success:bool)
   > Switch Screen(Widget=Get TileView_Rewards() -> (TileView_Rewards:Fort Tile View Object Reference))
   > SetVariantOverrides!(
    execute
    self:Fort Item Preview Widget Object Reference=Get FortItemPreviewWidget() -> (FortItemPreviewWidget:WBP Item Preview Object Reference)
    InVariantOverrides:Array of Mcp Variant Channel Info Structures=Make Array!() -> (Array:Array of Mcp Variant Channel Info Structures)
   ) -> (then)
   FireEventElementInteractedByName(execute=@n6.else, Element Name=Back, WidgetCallingFrom=@n7)
   > PopContentWidget!(execute) -> (then)
   n8 = "On Variant Item Engaged (WBP_UIKit_VariantSelector)"!() -> (
    OutputDelegate
    then
    VariantItem:Fort Variant Item VM Object Reference
    VariantChannel:Fort Variant Channel VM Object Reference
    AdditionalInfo:String
   ) {is=title K2Node_ComponentBoundEvent}
   > Refresh Info Box Visibility!(execute, self, InText:Text=Get UnlockRequirements!(self:Fort Variant Item VM Object Reference=@n8.VariantItem) -> (UnlockRequirements:Text)) -> (then)
   n9 = "On Variant Item Previewed (WBP_UIKit_VariantSelector)"!() -> (OutputDelegate, then, VariantItem:Fort Variant Item VM Object Reference) {is=title K2Node_ComponentBoundEvent}
   > Refresh Info Box Visibility(InText=Get UnlockRequirements(self=@n9.VariantItem))
   Event PreConstruct()
   n10 = GetConfigDataBoolWithDefault!(EntryName:String=show_compatibility_tags_in_about_popup, bDefaultValue:bool=false) -> (ReturnValue:bool)
   > Set PermitCompatibilityTagDisplay!(
    execute
    PermitCompatibilityTagDisplay:bool=@n10
    self:WBP UIKit Item Description Base Object Reference=Get ItemDescription() -> (ItemDescription:WBP UIKit Item Description MPL Object Reference)
   ) -> (then, Output_Get:bool)
   > SetCompatibilityVisibility!(execute, self:WBP UIKit Item Description Base Object Reference=Get ItemDescription(), Show:bool=@n10) -> (then, IsVisible:bool)
   Event On Activated()
   n11 = Get FortItemPreviewWidget()
   > ActivateWidget(self=@n11)
   > Fire Event Sub View Opened!(execute, View:Widget Object Reference=Self-Reference(), AdditionalInfo:Flow Analytics Additional Info Structure, bIsOverlay:bool=false) -> (then)
   > Set Rewards() {is=title K2Node_CallFunction}
   > Refresh Info Box Visibility()
   > PlayUpsellMusic!(execute, self, bPlay:bool=true) -> (then)
   Event On Deactivated()
   > DeactivateWidget(self=@n11)
   > PlayUpsellMusic(bPlay=false)
  ]}
 Get Purchase Visibility {nodes [
   n1 = Get Purchase Visibility!() -> (then, Is Cabined:bool=~, Has CMS Data:bool=~, Sub State:EMcpSubscriptionState Enum=Inactive) {is=title K2Node_FunctionEntry}
   n2 = NOT Boolean(A=@n1)
   n3 = "Equal (Enum)"(A:EMcpSubscriptionState Enum=@n1.Sub State, B:EMcpSubscriptionState Enum=Inactive)
   > Return Node!(
    execute
    NewParam:ESlateVisibility Enum=Select(
     Option 0:ESlateVisibility Enum=Collapsed
     Option 1:ESlateVisibility Enum=SelfHitTestInvisible
     Index=AND Boolean!(A=@n2, B=@n1.Has CMS Data, C:bool=@n3) -> (ReturnValue)
    ) -> (ReturnValue:ESlateVisibility Enum)
   )
  ]}
 GetPrimaryContentSetup {nodes [
   GetPrimaryContentSetup() {is=title K2Node_FunctionEntry}
   > Return Node!(
    execute
    ReturnValue:Primary Content Setup Structure=Make Primary Content Setup!(
     bShowBottomBar:bool=false
     bShowChatWidget:bool=false
     bEnableNewTextChat:bool=False
     bShowTopBar:bool=false
     bTabSupportsInterface:bool=False
    ) -> (PrimaryContentSetup:Primary Content Setup Structure) {is=title K2Node_MakeStruct}
   )
  ]}
 GetStyleButtonVisibility {nodes [
   n1 = GetStyleButtonVisibility!() -> (then, SelectedItem:Fort Item VM Object Reference) {is=title K2Node_FunctionEntry}
   n2 = "Equal (Object)"(A=GetActiveWidget(self=Get Switcher_ScreenView()), B=Get TileView_Rewards())
   n3 = Get bHasVariants!(self:Fort Item VM Object Reference=@n1) -> (bHasVariants:bool)
   > Return Node!(
    execute
    NewParam:ESlateVisibility Enum=Select(Option 0=Collapsed, Option 1=SelfHitTestInvisible, Index=AND Boolean!(A=@n3, B=@n2) -> (ReturnValue))
   )
  ]}
 PlayUpsellMusic {nodes [PlayUpsellMusic!() -> (then, bPlay:bool=~) {is=title K2Node_FunctionEntry}]}
 Refresh Info Box Visibility {nodes [
   n1 = Refresh Info Box Visibility!() -> (then, InText:Text) {is=title K2Node_FunctionEntry}
   n2 = TextIsEmpty(InText=@n1)
   n3 = Get InfoBox_Conflict() -> (InfoBox_Conflict:WBP UIKit Info Box Object Reference)
   > SetText(
    self:WBP UIKit Info Box Object Reference=@n3
    InText:Text=Select(
     Option 0:Text=@n1
     Option 1:Text=MakeLiteralText(Value="LOCTEXT(\"Additional styles unlock with additional months of an active Crew Subscription.\")")
     Index=@n2
    ) -> (ReturnValue:Text)
   )
   > SetShowIcon!(execute, self:WBP UIKit Info Box Object Reference=@n3, Show Icon:bool=@n2) -> (then)
   > SetBrushFromTexture!(
    execute
    self:Image Object Reference=Get Icon!(self:WBP UIKit Info Box Object Reference=@n3) -> (Icon:Image Object Reference)
    Texture:Texture 2D Object Reference=Select(
     Option 0:Texture 2D Object Reference
     Option 1:Texture 2D Object Reference=/Game/UI/Foundation/Textures/Icons/Heroes/Variants/T-Icon-Variant-32
     Index=@n2
    ) -> (ReturnValue:Texture 2D Object Reference)
    bMatchSize:bool=false
   ) -> (then)
  ]}
 Set Rewards {nodes [
   Set Rewards() {is=title K2Node_FunctionEntry}
   n1 = Get Rewards()
   > Set List Items!(execute, self:List View Object Reference=Get TileView_Rewards(), InListItems:Array of Object References=@n1) -> (then) {is=title K2Node_CallFunction}
   > Set SelectedItem!(
    execute
    SelectedItem:Fort Item VM Object Reference="Get (a copy)"(Array:Array of Fort Item VM Object References=@n1) -> (Output:Fort Item VM Object Reference)
    self:Fort Item Selection VM Object Reference=Get FortItemSelectionVM()
   ) -> (then, Output_Get:Fort Item VM Object Reference)
   n2 = Length(TargetArray=@n1)
   > Switch Screen(Widget=Select(
     Option 0:Widget Object Reference=Get TileView_Rewards()
     Option 1:Widget Object Reference=Get VBox_VariantSelector()
     Index="Equal (Integer)"(A=@n2, B=1)
    ) -> (ReturnValue:Widget Object Reference))
  ]}
 Subscribe Completed {nodes [Subscribe Completed() {is=title K2Node_FunctionEntry}]}
 Switch Screen {nodes [
   n1 = Switch Screen!() -> (then, Widget:Widget Object Reference) {is=title K2Node_FunctionEntry}
   > SetActiveWidget(self=Get Switcher_ScreenView(), Widget=@n1)
   > RequestRefreshFocus(self:Common Activatable Widget Object Reference)
   n2 = Get FortItemSelectionVM()
   > SetVisibility(
    self=Get Button_ItemDetails() -> (Button_ItemDetails:WBP UIKit Button Quiet Object Reference)
    InVisibility=GetStyleButtonVisibility!(self, SelectedItem:Fort Item VM Object Reference=Get SelectedItem(self=@n2)) -> (NewParam:ESlateVisibility Enum=Visible) {is=title K2Node_CallFunction}
   )
   n3 = GetActiveWidget(self=Get Switcher_ScreenView())
   n4 = Get TileView_Rewards()
   > SetVisibility(
    self=Get Overlay_Subscribe() -> (Overlay_Subscribe:Overlay Object Reference)
    InVisibility=Select(Option 0:ESlateVisibility Enum=Collapsed, Option 1:ESlateVisibility Enum=SelfHitTestInvisible, Index="Equal (Object)"(A=@n3, B=@n4)) -> (ReturnValue:ESlateVisibility Enum)
   )
  ]}
 UpdateUpsellMusic {nodes [UpdateUpsellMusic() {is=title K2Node_FunctionEntry}]}
]
variables [
 Rewards {type="array<FortItemVM>", default []}
 Subscribe Completed {}
]
widgets {
 info {
  inheritedWidgetCount=0
  namedSlotCount=1
  parentClass=&CommonActivatableWidget
  rootWidgetClass=&SafeZone
  widgetCount=25
 }
 widgets [
  SafeZone_All {widgetClassPath=&SafeZone, bIsVariable=true, widget=SafeZone_All}
  Overlay_All {widgetClassPath=&Overlay, parent=SafeZone_All, widget=Overlay_All, slot {class=SafeZoneSlot}}
  VerticalBox_LeftSideInfo {
   widgetClassPath=&VerticalBox
   bIsVariable=true
   parent=Overlay_All
   widget=VerticalBox_LeftSideInfo
   slot {class=OverlaySlot, properties {horizontalAlignment=HAlign_Fill, verticalAlignment=VAlign_Fill, padding=<ltrb 80 40 80 24>}}
  }
  SizeBox_Details {
   widgetClassPath=&SizeBox
   bIsVariable=true
   parent=VerticalBox_LeftSideInfo
   widget=SizeBox_Details
   properties {
    bOverride_MinDesiredHeight=true
    bOverride_WidthOverride=true
    heightOverride=308
    maxDesiredWidth=800
    minDesiredHeight=500
    widthOverride=800
   }
   slot {class=VerticalBoxSlot, properties {horizontalAlignment=HAlign_Left, padding=<ltrb 0 0 0 16>}}
  }
  ItemDescription {
   widgetClassPath=/FortniteCoreUI/Customization/MPL/WBP_UIKit_ItemDescription_MPL
   bIsVariable=true
   parent=SizeBox_Details
   widget=ItemDescription
   slot {class=SizeBoxSlot}
  }
  SizeBox_Items {
   widgetClassPath=&SizeBox
   parent=VerticalBox_LeftSideInfo
   widget=SizeBox_Items
   properties {bOverride_WidthOverride=true, widthOverride=800}
   slot {class=VerticalBoxSlot, properties {horizontalAlignment=HAlign_Left, padding=<ltrb 0 0 0 24>, size {sizeRule=Fill, value=1}}}
  }
  Switcher_ScreenView {
   widgetClassPath=&CommonVisibilitySwitcher
   bIsVariable=true
   parent=SizeBox_Items
   widget=Switcher_ScreenView
   properties {flowDirectionPreference=Culture}
   slot {class=SizeBoxSlot}
  }
  TileView_Rewards {
   widgetClassPath=/Script/FortniteUI.FortTileView
   bIsVariable=true
   parent=Switcher_ScreenView
   widget=TileView_Rewards
   properties {
    clipping=Inherit
    entryHeight=130
    entryWidgetClass=&/WBP_UIKit_ItemCard_CrewPreview.WBP_UIKit_ItemCard_CrewPreview_C
    entryWidth=130
    horizontalEntrySpacing=8
    numDesignerPreviewEntries=8
    verticalEntrySpacing=8
   }
   slot {class=CommonVisibilitySwitcherSlot}
  }
  VBox_VariantSelector {
   widgetClassPath=&VerticalBox
   bIsVariable=true
   parent=Switcher_ScreenView
   widget=VBox_VariantSelector
   slot {class=CommonVisibilitySwitcherSlot}
  }
  WBP_UIKit_VariantSelector {
   widgetClassPath=/FortUILibrary/UIKit/VariantSelector/WBP_UIKit_VariantSelector
   bIsVariable=true
   parent=VBox_VariantSelector
   widget=WBP_UIKit_VariantSelector
   properties {bIsFocusable=true, displayUnownedVariants=true, showAgeGateWarnings=true, showUnownedVariantsAsLocked=false}
   slot {class=VerticalBoxSlot, properties {size {sizeRule=Fill, value=1}}}
  }
  InfoBox_Conflict {
   widgetClassPath=/FortUILibrary/UIKit/InfoBox/WBP_UIKit_InfoBox
   bIsVariable=true
   parent=VBox_VariantSelector
   widget=InfoBox_Conflict
   slot {class=VerticalBoxSlot, properties {verticalAlignment=VAlign_Bottom}}
  }
  Text_CosmeticWarning {
   widgetClassPath=&CommonTextBlock
   parent=VerticalBox_LeftSideInfo
   widget=Text_CosmeticWarning
   properties {
    justification=Center
    mobileFontSizeMultiplier=1.2
    scrollStyle=/FortUILibrary/Styles/Typography/Foundation/TextScrollStyle-NoFade.TextScrollStyle-NoFade_C
    text=These cosmetic items grant no competitive advantage. Gliders and Contrails aren't useable in Save the World mode. Outfits and Wraps don't include weapons.
    wrapTextAt=1664
    font {
     bForceMonospaced=false
     bMaterialIsStencil=false
     fontMaterial=/Game/UI/Frontend/ItemShop/Art/M_UI_LegalDisclaimer_DetailsScreen_Text
     fontObject=&BurbankSmall
     letterSpacing=0
     monospacedWidth=1
     size=13
     skewAmount=0
     typefaceFontName=Bold
     outlineSettings {
      bApplyOutlineToDropShadows=false
      bMiteredCorners=false
      bSeparateFillAlpha=true
      outlineMaterial=null
      outlineSize=1
      outlineColor=<rgba 0 0 0 0.05>
     }
    }
    shadowOffset=<0 0>
   }
   slot {class=VerticalBoxSlot, properties {horizontalAlignment=HAlign_Center, verticalAlignment=VAlign_Bottom, padding=<ltrb 0 0 0 10>}}
  }
  Border_CameraFrame {
   widgetClassPath=&Border
   bIsVariable=true
   parent=Overlay_All
   widget=Border_CameraFrame
   properties {
    visibility=HitTestInvisible
    background {
     bIsDynamicallyLoaded=false
     drawAs=NoDrawType
     imageType=NoImage
     mirroring=NoMirror
     resourceName=None
     resourceObject=null
     tiling=NoTile
     imageSize=<32 32>
     margin=<ltrb 0 0 0 0>
     outlineSettings {
      bUseBrushTransparency=false
      roundingType=HalfHeightRadius
      width=0
      color {colorUseRule=UseColor_Specified, specifiedColor=<rgba 0 0 0 0>}
      cornerRadii=<0 0 0 0>
     }
     tintColor {colorUseRule=UseColor_Specified, specifiedColor=<rgba 1 1 1 1>}
     uVRegion {bIsValid=false, max=<0 0>, min=<0 0>}
    }
    padding=<ltrb 0 0 0 0>
   }
   slot {class=OverlaySlot, properties {horizontalAlignment=HAlign_Fill, verticalAlignment=VAlign_Fill, padding=<ltrb 700 100 40 100>}}
  }
  FortItemPreviewWidget {
   widgetClassPath=/FortniteCoreUI/Frontend/ItemPreview/WBP_ItemPreview
   bIsVariable=true
   parent=Border_CameraFrame
   widget=FortItemPreviewWidget
   properties {
    bCanDisplayUserAccessories=false
    bSupportsActivationFocus=false
    itemViewObject=&self:WidgetTree.FortItemPreviewWidget.ItemViewObject
    previewActionButtons [
     &self:WidgetTree.Button_PreviewAction
     &self:WidgetTree.Button_PreviewAction1
    ]
    previewSceneChangerTag {tagName=UI.ItemPreview.SceneChanger.FortniteCrew}
   }
   slot {class=BorderSlot, properties {padding=<ltrb 0 0 0 0>}}
  }
  Overlay_Subscribe {
   widgetClassPath=&Overlay
   bIsVariable=true
   parent=Overlay_All
   widget=Overlay_Subscribe
   slot {class=OverlaySlot, properties {horizontalAlignment=HAlign_Right, verticalAlignment=VAlign_Bottom, padding=<ltrb 0 0 80 160>}}
  }
  ConfirmPurchase_Button {
   widgetClassPath=/FortUILibrary/UIKit/Buttons/WBP_UIKit_Button_Loud
   bIsVariable=true
   parent=Overlay_Subscribe
   widget=ConfirmPurchase_Button
   properties {
    bIsFocusable=false
    bRequiresHold=true
    buttonSize=Large
    minHeight=56
    minWidth=632
    overrideSecondaryTextSizeMultiplier=0.5
    secondaryText=$11.99
    text=Confirm Purchase
    triggeringInputAction {dataTable=/FortUILibrary/Input/ItemShopInputActionDataTable, rowName=PurchaseCrewSetHold}
   }
   slot {class=OverlaySlot, properties {horizontalAlignment=HAlign_Right, verticalAlignment=VAlign_Center}}
  }
  WBP_ActionBarContainer {
   widgetClassPath=/Game/UI/Frontend/BottomBar/WBP_ActionBarContainer
   bIsVariable=true
   parent=Overlay_All
   widget=WBP_ActionBarContainer
   slot {class=OverlaySlot, properties {horizontalAlignment=HAlign_Right, verticalAlignment=VAlign_Bottom, padding=<ltrb 0 0 80 64>}}
  }
  HorizontalBox_ButtonPrompts {widgetClassPath=&HorizontalBox, namedSlotHost=WBP_ActionBarContainer, widget=HorizontalBox_ButtonPrompts}
  Button_PreviewAction {
   widgetClassPath=/FortUILibrary/UIKit/Buttons/WBP_UIKit_Button_Quiet
   bIsVariable=true
   parent=HorizontalBox_ButtonPrompts
   widget=Button_PreviewAction
   properties {
    bAutoUpdateFromInputAction=true
    bIsFocusable=false
    minWidth=48
    text=""
    visibility=Collapsed
    triggeringInputAction {dataTable=&InputActionDataTable, rowName=GenericFaceButtonLeft}
   }
   slot {class=HorizontalBoxSlot, properties {padding=<ltrb 0 0 24 0>}}
  }
  Button_PreviewAction1 {
   widgetClassPath=&WBP_UIKit_Button_Quiet
   bIsVariable=true
   parent=HorizontalBox_ButtonPrompts
   widget=Button_PreviewAction1
   properties {
    bAutoUpdateFromInputAction=true
    bIsFocusable=false
    minWidth=48
    text=""
    visibility=Collapsed
    triggeringInputAction {dataTable=&InputActionDataTable, rowName=GenericFaceButtonLeft}
   }
   slot {class=HorizontalBoxSlot, properties {padding=<ltrb 0 0 24 0>}}
  }
  CommonHardwareVisibilityBorder_0 {
   widgetClassPath=&CommonHardwareVisibilityBorder
   parent=HorizontalBox_ButtonPrompts
   widget=CommonHardwareVisibilityBorder_0
   properties {visibilityQuery {
     autoDescription=" ANY( Input.Gamepad )"
     tokenStreamVersion=0
     userDescription=""
     queryTokenStream [0,1,1,1,0]
     tagDictionary [{tagName=Input.Gamepad}]
    }}
   slot {class=HorizontalBoxSlot, properties {padding=<ltrb 0 0 24 0>}}
  }
  Button_Camera {
   widgetClassPath=&WBP_UIKit_Button_Quiet
   bIsVariable=true
   parent=CommonHardwareVisibilityBorder_0
   widget=Button_Camera
   properties {bIsFocusable=false, text=Camera, triggeringInputAction {dataTable=&InputActionDataTable, rowName=UnifiedRotateItem}}
   slot {class=BorderSlot, properties {padding=<ltrb 0 0 0 0>}}
  }
  Button_ItemDetails {
   widgetClassPath=&WBP_UIKit_Button_Quiet
   bIsVariable=true
   parent=HorizontalBox_ButtonPrompts
   widget=Button_ItemDetails
   properties {
    bIsFocusable=false
    text="Details && Styles"
    triggeringInputAction {dataTable=&InputActionDataTable, rowName=GenericForward}
   }
   slot {class=HorizontalBoxSlot, properties {padding=<ltrb 0 0 24 0>}}
  }
  Button_Back {
   widgetClassPath=&WBP_UIKit_Button_Quiet
   bIsVariable=true
   parent=HorizontalBox_ButtonPrompts
   widget=Button_Back
   properties {bIsFocusable=false, text=Back, triggeringInputAction {dataTable=&InputActionDataTable, rowName=GenericBack}}
   slot {class=HorizontalBoxSlot}
  }
  WBP_PlayerInfo {
   widgetClassPath=/Game/UI/Frontend/ItemShop/MultiProduct/WBP_PlayerInfo
   parent=Overlay_All
   widget=WBP_PlayerInfo
   slot {class=OverlaySlot, properties {horizontalAlignment=HAlign_Right, padding=<ltrb 0 40 80 0>}}
  }
 ]
}
